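# UI-domain material for the Billy shop tile in the item shop.
# It samples the Billy tile texture using scale-to-fill UVs from the UV scaling function, then lerps the color toward a constant dark green.
# The lerp alpha comes from a rotated linear gradient function whose range starts at the midpoint, fading the artwork into the flat color.
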Material /ItemShop/UI/Art/Billy/M_UI_Billy_ShopTile
graph {
 properties {bUsedWithStaticMesh=false, materialDomain=MD_UI, pixelDepthOffsetMode=PDOM_Legacy, usedShadingModels=Default Lit}
 nodes [LinearInterpolate(
   A=TextureSample(UVs=MaterialFunctionCall() {materialFunction=&MF_UI_UVScaling}.Scaled To Fill) {texture=/OfferCatalog/Art/Billy/T_UI_Billy_ShopTile}.RGB
   B=Constant3Vector() {constant=<rgba 0 0.0781 0.0213 0>}
   Alpha=MaterialFunctionCall(GradientStart=Constant() {r=0.5}, GradientEnd=Constant() {r=1}, RotationAngle=Constant() {r=0.75}) {materialFunction=&MF_UI_LinearGradient}.Result
  )]
}
